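# DragonCart_100 is a Level Sequence in DragonCartNarrative that acts as a container, apparently for one shot of the dragon cart narrative.
# It has no bindings of its own, only a Subsequences track with seven sub-sections.
# All seven share the same 1.5-second range, frames 9 to 99 at 60 fps, so they play in parallel.
# The file does not show which sequences they reference.
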LevelSequence /DragonCartNarrative/Sequence/DragonCart_100/DragonCart_100
bindings []
display_rate {denominator=1, numerator=60}
tick_resolution {denominator=1, numerator=24000}
tracks [{
  type=MovieSceneSubTrack
  display_name=Subsequences
  sections [
   {type=MovieSceneSubSection, end_frame=39600, start_frame=3600}
   {type=MovieSceneSubSection, end_frame=39600, start_frame=3600}
   {type=MovieSceneSubSection, end_frame=39600, start_frame=3600}
   {type=MovieSceneSubSection, end_frame=39600, start_frame=3600}
   {type=MovieSceneSubSection, end_frame=39600, start_frame=3600}
   {type=MovieSceneSubSection, end_frame=39600, start_frame=3600}
   {type=MovieSceneSubSection, end_frame=39600, start_frame=3600}
  ]
 }]
