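# DragonCart_100_cam is a Level Sequence in the DragonCartNarrative content, named as the camera sequence for DragonCart_100.
# It holds no bindings and no tracks, so it animates nothing as it stands.
# It is set to a 60 fps display rate with a 24000 tick resolution.
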LevelSequence /DragonCartNarrative/Sequence/DragonCart_100/DragonCart_100_cam
bindings []
display_rate {denominator=1, numerator=60}
tick_resolution {denominator=1, numerator=24000}
tracks []
