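# Level Sequence for shot DragonCart_100 in the DragonCart narrative, apparently its technical pass, that drives World Partition data layer states.
# It has no bindings, only paired Activate and Unload Data Layer tracks, with the target layers not shown.
# Most tracks share an opening window, then the last Activate sections step through consecutive short windows.
# Complementary Unload sections surround these windows, suggesting one layer set hands off to the next.
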LevelSequence /DragonCartNarrative/Sequence/DragonCart_100/DragonCart_100_tec
bindings []
display_rate {denominator=1, numerator=60}
tick_resolution {denominator=1, numerator=24000}
tracks [
 {type=MovieSceneDataLayerTrack, display_name=Activate, sections [{type=MovieSceneDataLayerSection, end_frame=7200, start_frame=3600}]}
 {type=MovieSceneDataLayerTrack, display_name=Unload, sections [{type=MovieSceneDataLayerSection, end_frame=7200, start_frame=3600}]}
 {type=MovieSceneDataLayerTrack, display_name=Activate, sections [{type=MovieSceneDataLayerSection, end_frame=7200, start_frame=3600}]}
 {type=MovieSceneDataLayerTrack, display_name=Unload, sections [{type=MovieSceneDataLayerSection, end_frame=7200, start_frame=3600}]}
 {type=MovieSceneDataLayerTrack, display_name=Activate, sections [{type=MovieSceneDataLayerSection, end_frame=7200, start_frame=3600}]}
 {type=MovieSceneDataLayerTrack, display_name=Unload, sections [{type=MovieSceneDataLayerSection, end_frame=7200, start_frame=3600}]}
 {type=MovieSceneDataLayerTrack, display_name=Activate, sections [{type=MovieSceneDataLayerSection, end_frame=7200, start_frame=3600}]}
 {type=MovieSceneDataLayerTrack, display_name=Unload, sections [{type=MovieSceneDataLayerSection, end_frame=7200, start_frame=3600}]}
 {type=MovieSceneDataLayerTrack, display_name=Activate, sections [{type=MovieSceneDataLayerSection, end_frame=7200, start_frame=3600}]}
 {type=MovieSceneDataLayerTrack, display_name=Unload, sections [{type=MovieSceneDataLayerSection, end_frame=25200, start_frame=7200}]}
 {type=MovieSceneDataLayerTrack, display_name=Activate, sections [{type=MovieSceneDataLayerSection, end_frame=10800, start_frame=7200}]}
 {
  type=MovieSceneDataLayerTrack
  display_name=Unload
  sections [{type=MovieSceneDataLayerSection, end_frame=25200, start_frame=10800}, {type=MovieSceneDataLayerSection, end_frame=7200, start_frame=3600}]
 }
 {type=MovieSceneDataLayerTrack, display_name=Activate, sections [{type=MovieSceneDataLayerSection, end_frame=14400, start_frame=10800}]}
 {
  type=MovieSceneDataLayerTrack
  display_name=Unload
  sections [{type=MovieSceneDataLayerSection, end_frame=25200, start_frame=14400}, {type=MovieSceneDataLayerSection, end_frame=10800, start_frame=3600}]
 }
]
